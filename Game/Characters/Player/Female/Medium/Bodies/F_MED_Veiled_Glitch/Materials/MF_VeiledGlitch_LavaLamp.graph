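# Material function that adds an animated-style lava lamp glow to the Veiled Glitch character's emissive and blacks out base color.
# Sine-warped curve noise samples a gradient curve atlas, blended over dimmed base color by sharpened visibility noise.
# Inverted Fresnel and a bottom-to-top intensity gradient from the Position input's Z scale the glow.
# A time-of-day multiply is present but not connected to the output.
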MaterialFunction /Game/Characters/Player/Female/Medium/Bodies/F_MED_Veiled_Glitch/Materials/MF_VeiledGlitch_LavaLamp
graph {nodes [
  n1 = GetMaterialAttributes(None=FunctionInput() {inputName=Material, inputType=FunctionInput_MaterialAttributes}) {attributeGetTypes [#1, #2]}
  n2 = Multiply(A=@n1.BaseColor, B=ScalarParameter() {defaultValue=0.45, parameterName=Base Emissive})
  n3 = FunctionInput() {inputName=Curve Noise}
  n4 = Sine(None=Multiply(A=ScalarParameter() {defaultValue=1, parameterName=Warp Freq}, B=Power(Base=@n3) {constExponent=3}))
  n5 = CurveAtlasRowParameter(CurveTime=ComponentMask(None=Saturate(None=@n4)) {r=true}) {
   atlas=/Game/Characters/Shared/CurveAtlas/AthenaCharacterCurveAtlas
   curve=&/VeiledGlitchGradient
   defaultValue=32
  }
  n6 = FunctionInput() {inputName=Visibility noise}
  n7 = ScalarParameter() {defaultValue=0.8, parameterName=Freq bias}
  n8 = Saturate(None=Power(Base=Multiply(A=@n6, B=@n7), Exp=ScalarParameter() {defaultValue=4, parameterName=Freq}))
  n9 = OneMinus(None=Power(
    Base=FunctionInput() {inputName=Fresnel, inputType=FunctionInput_Scalar}
    Exp=ScalarParameter() {defaultValue=0.25, parameterName=Fresnel Power}
   ))
  n10 = Multiply(
   A=LinearInterpolate(
    A=ScalarParameter() {defaultValue=12, parameterName=Emissive Intensity Bottom}
    B=ScalarParameter() {defaultValue=3, parameterName=Emissive Intensity Top}
    Alpha=NamedRerouteUsage()
   )
   B=Multiply(A=LinearInterpolate(A=@n2, B=@n5, Alpha=@n8), B=Clamp(None=@n9) {maxDefault=99, minDefault=0.05})
  )
  FunctionOutput(None=SetMaterialAttributes!(MaterialAttributes=@n1.MaterialAttributes, EmissiveColor=Add(A=@n1.EmissiveColor, B=@n10), BaseColor=Constant()) {attributeSetTypes [#1, #2]})
  n11 = Constant() {r=0.5}
  Multiply(
   A=@n10
   B=MaterialFunctionCall(Value_Morning=@n11, Value_Day=@n11, Value_Evening=@n11, Value_Night=Constant() {r=1}) {materialFunction=&Scalar_Time_of_Day}.Result
  )
  n12 = MaterialFunctionCall(Float3=FunctionInput() {inputName=Position}) {materialFunction=&BreakOutFloat3Components}
  NamedRerouteDeclaration(None=Clamp(
    None=Divide(A=@n12.B) {constB=150}
    Min=ScalarParameter() {defaultValue=0.15, parameterName=Bottom Color}
    Max=ScalarParameter() {defaultValue=0.95, parameterName=Top Color}
   )) {name=Top to Bottom gradient, nodeColor=<rgba 0.2422 0 1 1>}
 ]}
