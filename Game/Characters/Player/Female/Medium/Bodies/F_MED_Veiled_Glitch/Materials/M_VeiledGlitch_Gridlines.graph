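# Material function for the Veiled Glitch player body that adds glowing gridlines to the emissive color of incoming material attributes.
# The Gridlines input is tinted by an HDR color parameter and added to the existing emissive.
# A lerp reveals that result through a mask built from scaled, contrast-boosted Noise, multiplied by inverted, power-shaped Fresnel.
# Only EmissiveColor is overwritten, with color, frequency, contrast, boost and Fresnel power exposed as parameters.
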MaterialFunction /Game/Characters/Player/Female/Medium/Bodies/F_MED_Veiled_Glitch/Materials/M_VeiledGlitch_Gridlines
graph {nodes [
  Constant()
  n1 = GetMaterialAttributes(None=FunctionInput() {inputType=FunctionInput_MaterialAttributes}) {attributeGetTypes [#1, #2]}
  n2 = Add(
   A=@n1.EmissiveColor
   B=Multiply(
    A=VectorParameter() {parameterName=Gridline Color, defaultValue=<rgba 5 1.3125 3.8628 1>}.RGB
    B=FunctionInput() {inputName=Gridlines, inputType=FunctionInput_Scalar}
   )
  )
  n3 = MaterialFunctionCall(
   In=Multiply(A=FunctionInput() {inputName=Noise, inputType=FunctionInput_Scalar}, B=ScalarParameter() {defaultValue=1, parameterName=Gridline Frequency})
   Contrast=ScalarParameter() {defaultValue=3, parameterName=Gridline Contrast}
  ) {materialFunction=&CheapContrast}
  n4 = ScalarParameter() {defaultValue=1, parameterName=Gridline Boost}
  n5 = Power(
   Base=FunctionInput() {inputName=Fresnel, inputType=FunctionInput_Scalar}
   Exp=ScalarParameter() {defaultValue=0.2, parameterName=Fresnel Power}
  )
  n6 = Multiply(A=Saturate(None=Multiply(A=@n3.Result, B=@n4)), B=Saturate(None=OneMinus(None=@n5)))
  FunctionOutput(None=SetMaterialAttributes!(MaterialAttributes=@n1.MaterialAttributes, EmissiveColor=LinearInterpolate(A=@n1.EmissiveColor, B=@n2, Alpha=@n6)) {attributeSetTypes [#1]})
 ]}
