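# Parent skeletal-mesh material for the Veiled Glitch female character body, layering animated lava-lamp and gridline effects between the shared character PreFX and PostFX functions.
# Two panning Perlin projections in pre-skinned local space, one warped by the other, build visibility and curve noise for the lava-lamp function.
# Its result feeds the gridline function, masked by SkinFX green.
# A clamped, parameter-scaled Fresnel transition glow is added to emissive before PostFX.
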Material /Game/Characters/Player/Female/Medium/Bodies/F_MED_Veiled_Glitch/Materials/M_F_MED_Vieled_Glitch_Parent
graph {
 properties {
  bUseMaterialAttributes=true
  bUsedWithMeshDeformer=true
  bUsedWithMorphTargets=true
  bUsedWithSkeletalMesh=true
  pixelDepthOffsetMode=PDOM_Legacy
  usedShadingModels=Default Lit
 }
 nodes [
  n1 = MaterialFunctionCall() {materialFunction=&BaseFN_CharacterFunction_PreFX}
  n2 = TextureSampleParameter2D() {
   parameterName=SkinFX
   samplerType=SAMPLERTYPE_Masks
   texture=/Game/Characters/Player/Female/Medium/Bodies/F_MED_Veiled_Glitch/Textures/T_F_MED_Veiled_Glitch_Body_FX
  }
  n3 = TextureObject() {samplerType=SAMPLERTYPE_Masks, texture=/Game/Global/Textures/Default/Noise/T_Generic_Perlin}
  n4 = VectorParameter() {parameterName="Panning  2", defaultValue=<rgba 0 0.1 0 1>}
  n5 = MaterialFunctionCall!(
   TextureObject=@n3
   TextureSize=ScalarParameter() {defaultValue=155, parameterName=Noise Size 3}
   WorldPosition=@n1.Pre skinned Local Position
   Export Float 4
   In=MaterialFunctionCall(X=@n4.R, Y=@n4.R) {materialFunction=&MakeFloat2}.Result
   World Space Normal
   ProjectionTransitionContrast
  ) {materialFunction=/Game/Athena/SafeZone/WorldAlignedTexture_ScalePan}
  n6 = ScalarParameter() {defaultValue=55, parameterName=Texture Warp Amount}
  n7 = VectorParameter() {parameterName=Panning Speed, defaultValue=<rgba 0 0.15 0 1>}
  n8 = MaterialFunctionCall(
   TextureObject=@n3
   TextureSize=ScalarParameter() {defaultValue=200, parameterName=Noise Size}
   WorldPosition=Add(A=@n1.Pre skinned Local Position, B=Multiply(A=@n5.XY Texture, B=@n6))
   In=MaterialFunctionCall(X=@n7.R, Y=@n7.R) {materialFunction=&MakeFloat2}.Result
  ) {materialFunction=&WorldAlignedTexture_ScalePan}
  n9 = ComponentMask(None=@n5.XY Texture) {g=true}
  n10 = Multiply(A=ComponentMask(None=@n8.XY Texture) {g=true}, B=Add(A=@n9) {constB=0.5})
  n11 = ComponentMask(None=@n5.XY Texture) {g=true}
  n12 = Subtract(A=Multiply(A=ComponentMask(None=@n8.XY Texture) {r=true}, B=Add(A=@n11) {constB=0.5}), B=Power(Base=@n9) {constExponent=4})
  n13 = GetMaterialAttributes(None=MaterialFunctionCall!(
    Gridlines=@n2.G
    Noise=ComponentMask(None=@n5.XY Texture) {b=true}
    Fresnel=@n1.Fresnel
    In=MaterialFunctionCall!(
     Position=@n1.Pre skinned Local Position
     Fresnel=@n1.Fresnel
     Visibility noise=@n10
     Curve Noise=@n12
     Material=@n1.Pre skinned Local Position
    ) {materialFunction=&/MF_VeiledGlitch_LavaLamp}.Result
   ) {materialFunction=&/M_VeiledGlitch_Gridlines}.Result) {attributeGetTypes [#1]}
  n14 = VectorParameter() {parameterName=Transition Glow, defaultValue=<rgba 0 1.6 4 1>}
  n15 = Power(Base=@n1.Fresnel) {constExponent=2.5}
  n16 = Clamp(None=Multiply(A=Multiply(A=@n14.RGB, B=@n15), B=ScalarParameter() {parameterName=VeiledTransitionGlowAmount})) {maxDefault=20}
  MaterialFunctionCall(
   AO=@n1.AO
   Distance Blend=@n1.AO
   Fresnel=@n1.AO
   Gmap=@n1.AO
   Pre Skinned Local Position=@n1.AO
   Material=SetMaterialAttributes(MaterialAttributes=@n13.MaterialAttributes, Emissive Color=Add(A=@n13.Emissive Color, B=@n16)) {attributeSetTypes [#1]}
  ) {materialFunction=&BaseFN_CharacterFunction_PostFX}
  Subtract(B=Divide(A=@n11) {constB=3})
 ]
}
