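# Physics asset of collision shapes for the F_MED_Veiled_Glitch player body, apparently for its simulated cape to collide against.
# Capsules cover head, clavicles, spine, pelvis and feet, with some spine bones doubling up using rotated capsules.
# Both hip ribbon chains carry capsules plus small base spheres but have no constraints.
# Arm, thigh, calf and neck bodies hold no shapes, and all body-chain constraints leave swing and twist free.
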PhysicsAsset /Game/Characters/Player/Female/Medium/Bodies/F_MED_Veiled_Glitch/Meshes/F_MED_Veiled_Glitch_CapeColliders
bodies {
 calf_l []
 calf_r []
 clavicle_l [{shapeType=Capsule, length=7.6964, radius=6.3853, center=<4.0147 -0.8354 -1.7419>, rotation=<rot -77.5566 177.3915 7.7314>}]
 clavicle_r [{shapeType=Capsule, length=7.6964, radius=8.2957, center=<-4.0147 0.8354 1.7419>, rotation=<rot -77.5565 177.3915 -172.2685>}]
 dyn_hip_ribbon_01_l [{shapeType=Capsule, length=15.5118, radius=6.0926, shapeName=dyn_hip_ribbon_01_l_capsule, center=<5.8824 0 -2.0042>, rotation=<rot 90 0 0>}]
 dyn_hip_ribbon_01_r [{shapeType=Capsule, length=14.6853, radius=5.7388, shapeName=dyn_hip_ribbon_01_r_capsule, center=<5.4219 0 -2.0042>, rotation=<rot 90 0 0>}]
 dyn_hip_ribbon_02_l [{shapeType=Capsule, length=18.7424, radius=6.0533, shapeName=dyn_hip_ribbon_02_l_capsule, center=<5.7275 0 -2.0042>, rotation=<rot 90 0 0>}]
 dyn_hip_ribbon_02_r [{shapeType=Capsule, length=16.1777, radius=6.028, shapeName=dyn_hip_ribbon_02_r_capsule, center=<5.7993 0 -2.0042>, rotation=<rot 90 0 0>}]
 dyn_hip_ribbon_03_l [{shapeType=Capsule, length=18.3268, radius=6.5584, shapeName=dyn_hip_ribbon_03_l_capsule, center=<6.7714 0 -2.0042>, rotation=<rot 90 0 0>}]
 dyn_hip_ribbon_03_r [{shapeType=Capsule, length=19.0537, radius=5.9254, shapeName=dyn_hip_ribbon_03_r_capsule, center=<5.739 0 -2.0042>, rotation=<rot 90 0 0>}]
 dyn_hip_ribbon_base_l [{shapeType=Sphere, radius=1.2738, shapeName=dyn_hip_ribbon_base_l_sphere}]
 dyn_hip_ribbon_base_r [{shapeType=Sphere, radius=1.2738, shapeName=dyn_hip_ribbon_base_r_sphere}]
 dyn_hip_ribbon_parent_l [{
   shapeType=Capsule
   length=13.1743
   radius=7.0835
   shapeName=dyn_hip_ribbon_parent_l_capsule
   center=<5.9366 0 -2.0042>
   rotation=<rot 90 0 0>
  }]
 dyn_hip_ribbon_parent_r [{
   shapeType=Capsule
   length=13.5237
   radius=7.0467
   shapeName=dyn_hip_ribbon_parent_r_capsule
   center=<4.3971 0 -2.0042>
   rotation=<rot 90 0 0>
  }]
 foot_l [{shapeType=Capsule, length=17.869, radius=11.4605, center=<-2.2072 6.0534 -0.3367>, rotation=<rot 0.0388 1.4701 -89.3982>}]
 foot_r [{shapeType=Capsule, length=20.7585, radius=10.1705, center=<3.8217 -7.2009 0.3502>, rotation=<rot 0.0388 1.4701 90.6018>}]
 head [{shapeType=Capsule, length=0, radius=18, center=<9.0697 0.0746 0>, rotation=<rot 0 -90 0>}]
 lowerarm_l []
 lowerarm_r []
 neck_02 []
 pelvis [{shapeType=Capsule, length=5.7367, radius=14.3256, center=<-1.9544 -3.0824 -0.1981>}]
 spine_01 [{shapeType=Capsule, length=5.3511, radius=13.4014, center=<-1.733 -1.9668 0>}]
 spine_02 [
  {shapeType=Capsule, length=3.5443, radius=12.1514, center=<1.179 -0.641 0>}
  {shapeType=Capsule, length=20.6875, radius=9.3125, center=<-1.8317 2.606 0>, rotation=<rot 90 0 0>}
 ]
 spine_03 [
  {shapeType=Capsule, length=6.3241, radius=9.6328, center=<1.8605 -1.558 0>}
  {shapeType=Capsule, length=20.7188, radius=9.2812, center=<-1.2326 2.6059 0>, rotation=<rot 90 0 0>}
 ]
 spine_04 [
  {shapeType=Capsule, length=5.8296, radius=10.3916, center=<5.5182 1.4112 0>}
  {shapeType=Capsule, length=6.3652, radius=12.9473, center=<-1.2326 0.4285 0>, rotation=<rot 90 0 0>}
 ]
 spine_05 [{shapeType=Capsule, length=10.9859, radius=10.0791, center=<5.6377 0.3227 0>}]
 thigh_l []
 thigh_r []
 upperarm_l []
 upperarm_r []
}
constraints [
 {
  bone1Name=spine_01
  bone2Name=pelvis
  swing1LimitDegrees=45
  swing1Motion=Free
  swing2LimitDegrees=45
  swing2Motion=Free
  twistLimitDegrees=45
  twistMotion=Free
 }
 {
  bone1Name=spine_02
  bone2Name=spine_01
  swing1LimitDegrees=45
  swing1Motion=Free
  swing2LimitDegrees=45
  swing2Motion=Free
  twistLimitDegrees=45
  twistMotion=Free
 }
 {
  bone1Name=spine_03
  bone2Name=spine_02
  swing1LimitDegrees=45
  swing1Motion=Free
  swing2LimitDegrees=45
  swing2Motion=Free
  twistLimitDegrees=45
  twistMotion=Free
 }
 {
  bone1Name=spine_04
  bone2Name=spine_03
  swing1LimitDegrees=45
  swing1Motion=Free
  swing2LimitDegrees=45
  swing2Motion=Free
  twistLimitDegrees=45
  twistMotion=Free
 }
 {
  bone1Name=spine_05
  bone2Name=spine_04
  swing1LimitDegrees=45
  swing1Motion=Free
  swing2LimitDegrees=45
  swing2Motion=Free
  twistLimitDegrees=45
  twistMotion=Free
 }
 {
  bone1Name=clavicle_l
  bone2Name=spine_05
  swing1LimitDegrees=45
  swing1Motion=Free
  swing2LimitDegrees=45
  swing2Motion=Free
  twistLimitDegrees=45
  twistMotion=Free
 }
 {
  bone1Name=upperarm_l
  bone2Name=clavicle_l
  swing1LimitDegrees=45
  swing1Motion=Free
  swing2LimitDegrees=45
  swing2Motion=Free
  twistLimitDegrees=45
  twistMotion=Free
 }
 {
  bone1Name=lowerarm_l
  bone2Name=upperarm_l
  swing1LimitDegrees=45
  swing1Motion=Free
  swing2LimitDegrees=45
  swing2Motion=Free
  twistLimitDegrees=45
  twistMotion=Free
 }
 {
  bone1Name=neck_02
  bone2Name=spine_05
  swing1LimitDegrees=45
  swing1Motion=Free
  swing2LimitDegrees=45
  swing2Motion=Free
  twistLimitDegrees=45
  twistMotion=Free
 }
 {
  bone1Name=head
  bone2Name=neck_02
  swing1LimitDegrees=45
  swing1Motion=Free
  swing2LimitDegrees=45
  swing2Motion=Free
  twistLimitDegrees=45
  twistMotion=Free
 }
 {
  bone1Name=calf_l
  bone2Name=thigh_l
  swing1LimitDegrees=45
  swing1Motion=Free
  swing2LimitDegrees=45
  swing2Motion=Free
  twistLimitDegrees=45
  twistMotion=Free
 }
 {
  bone1Name=thigh_l
  bone2Name=pelvis
  swing1LimitDegrees=45
  swing1Motion=Free
  swing2LimitDegrees=45
  swing2Motion=Free
  twistLimitDegrees=45
  twistMotion=Free
 }
 {
  bone1Name=foot_l
  bone2Name=calf_l
  swing1LimitDegrees=45
  swing1Motion=Free
  swing2LimitDegrees=45
  swing2Motion=Free
  twistLimitDegrees=45
  twistMotion=Free
 }
 {
  bone1Name=clavicle_r
  bone2Name=spine_05
  swing1LimitDegrees=45
  swing1Motion=Free
  swing2LimitDegrees=45
  swing2Motion=Free
  twistLimitDegrees=45
  twistMotion=Free
 }
 {
  bone1Name=lowerarm_r
  bone2Name=upperarm_r
  swing1LimitDegrees=45
  swing1Motion=Free
  swing2LimitDegrees=45
  swing2Motion=Free
  twistLimitDegrees=45
  twistMotion=Free
 }
 {
  bone1Name=thigh_r
  bone2Name=pelvis
  swing1LimitDegrees=45
  swing1Motion=Free
  swing2LimitDegrees=45
  swing2Motion=Free
  twistLimitDegrees=45
  twistMotion=Free
 }
 {
  bone1Name=calf_r
  bone2Name=thigh_r
  swing1LimitDegrees=45
  swing1Motion=Free
  swing2LimitDegrees=45
  swing2Motion=Free
  twistLimitDegrees=45
  twistMotion=Free
 }
 {
  bone1Name=foot_r
  bone2Name=calf_r
  swing1LimitDegrees=45
  swing1Motion=Free
  swing2LimitDegrees=45
  swing2Motion=Free
  twistLimitDegrees=45
  twistMotion=Free
 }
 {
  bone1Name=upperarm_r
  bone2Name=clavicle_r
  swing1LimitDegrees=45
  swing1Motion=Free
  swing2LimitDegrees=45
  swing2Motion=Free
  twistLimitDegrees=45
  twistMotion=Free
 }
]
