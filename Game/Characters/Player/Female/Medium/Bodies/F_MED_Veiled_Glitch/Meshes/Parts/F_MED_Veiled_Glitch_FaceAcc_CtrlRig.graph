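# Control Rig for the face accessory part of the female medium Veiled Glitch player body, adjusting its dynamic hair bones.
# The executed chain zero-scales the first left and right hair bones, likely hiding the strands, then rotates both hair bases and shifts the right one.
# Unconnected leftovers include ear chain harmonics with pendulum and wave, an alternative hair aim and squash chain, and random float smoothing.
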ControlRigBlueprint /Game/Characters/Player/Female/Medium/Bodies/F_MED_Veiled_Glitch/Meshes/Parts/F_MED_Veiled_Glitch_FaceAcc_CtrlRig
graphs [
 RigVMFunctionLibrary {links [], nodes []}
 RigVMModel {nodes [
   /Script/ControlRig.RigUnit_ChainHarmonicsPerItem(
    ChainRoot=(Type=Bone,Name="dyn_ear_r_base")
    DrawWorldOffset=(Rotation=(X=0,Y=0,Z=0,W=1),Translation=(X=0,Y=0,Z=0),Scale3D=(X=1,Y=1,Z=1))
    ExecutePin=()
    Pendulum=(bEnabled=True,PendulumStiffness=2,PendulumGravity=(X=0,Y=0,Z=0),PendulumBlend=0.75,PendulumDrag=0.98,PendulumMaximum=0.1,UnwindAxis=(X=0,Y=1,Z=0),UnwindMinimum=0.2,UnwindMaximum=0.05)
    Reach=(bEnabled=false,ReachTarget=(X=1,Y=0,Z=0),ReachAxis=(X=1,Y=0,Z=0),ReachMinimum=0,ReachMaximum=1,ReachEase=SineEaseInOut)
    Speed=(X=0.25,Y=0.25,Z=0)
    Wave=(bEnabled=True,WaveFrequency=(X=20,Y=0,Z=0),WaveAmplitude=(X=0,Y=1,Z=0),WaveOffset=(X=1,Y=0,Z=0),WaveNoise=(X=10,Y=0,Z=0),WaveMinimum=0.1,WaveMaximum=1,WaveEase=SineEaseInOut)
    WaveCurve=(EditorCurveData=(Keys=((Time=0.0011,Value=0.0013),(Time=0.9761,Value=0.4035))))
    bDrawDebug=True
   ) {is=struct RigVMUnitNode}
   > &RigUnit_ChainHarmonicsPerItem(
    ChainRoot=(Type=Bone,Name="dyn_ear_l_base")
    DrawWorldOffset=(Rotation=(X=0,Y=0,Z=0,W=1),Translation=(X=0,Y=0,Z=0),Scale3D=(X=1,Y=1,Z=1))
    Pendulum=(bEnabled=True,PendulumStiffness=2,PendulumGravity=(X=0,Y=0,Z=0),PendulumBlend=0.75,PendulumDrag=0.98,PendulumMaximum=0.1,UnwindAxis=(X=0,Y=1,Z=0),UnwindMinimum=0.2,UnwindMaximum=0.05)
    Reach=(bEnabled=false,ReachTarget=(X=1,Y=0,Z=0),ReachAxis=(X=1,Y=0,Z=0),ReachMinimum=0,ReachMaximum=1,ReachEase=SineEaseInOut)
    Speed=(X=0.3,Y=0.3,Z=0)
    Wave=(bEnabled=True,WaveFrequency=(X=10,Y=0,Z=0),WaveAmplitude=(X=0,Y=2,Z=0),WaveOffset=(X=2,Y=0,Z=0),WaveNoise=(X=20,Y=0,Z=0),WaveMinimum=0.1,WaveMaximum=1,WaveEase=SineEaseInOut)
    WaveCurve=(EditorCurveData=(Keys=((Time=0.0011,Value=0.0013),(Time=0.9761,Value=0.4035))))
    bDrawDebug=True
   )
   RigUnit_BeginExecution = &RigUnit_BeginExecution()
   &RigUnit_ModifyTransforms(
    ItemToModify=((Item=(Type=Bone,Name="dyn_hair_l_01"),Transform=(Rotation=(X=0,Y=0,Z=0,W=1),Translation=(X=0,Y=0,Z=0),Scale3D=(X=0,Y=0,Z=0))))
    Weight=1
    ExecutePin=@RigUnit_BeginExecution.ExecutePin
   )
   > &RigUnit_ModifyTransforms(
    ItemToModify=((Item=(Type=Bone,Name="dyn_hair_r_01"),Transform=(Rotation=(X=0,Y=0,Z=0,W=1),Translation=(X=0,Y=0,Z=0),Scale3D=(X=0,Y=0,Z=0))))
    Weight=1
   )
   > &RigUnit_ModifyTransforms(
    ItemToModify=((Item=(Type=Bone,Name="dyn_hair_l_base"),Transform=(Rotation=(X=0,Y=0,Z=0,W=1),Translation=(X=0,Y=0,Z=0),Scale3D=(X=1,Y=1,Z=1))))
    Weight=1
    ItemToModify.0.Transform.Rotation=&RigVMFunction_MathQuaternionFromEuler(Euler=(X=0,Y=-23,Z=10)).Result
   )
   > &RigUnit_ModifyTransforms(
    ItemToModify=((Item=(Type=Bone,Name="dyn_hair_r_base"),Transform=(Rotation=(X=0,Y=0,Z=0,W=1),Translation=(X=0,Y=0,Z=-3),Scale3D=(X=1,Y=1,Z=1))))
    Weight=1
    ItemToModify.0.Transform.Rotation=&RigVMFunction_MathQuaternionFromEuler(Euler=(X=0,Y=-20,Z=-10)).Result
   )
   &RigUnit_SetTransform(
    ExecutePin=()
    Item=(Type=Bone,Name="const_hair_aim")
    Space=LocalSpace
    Value=(Rotation=(X=0,Y=0,Z=0,W=1),Translation=(X=0,Y=-13,Z=0),Scale3D=(X=1,Y=1,Z=1))
    Weight=1
    bInitial=false
   )
   > /Script/ControlRig.RigUnit_AimItem(
    DebugSettings=(Scale=10,WorldOffset=(Rotation=(X=0,Y=0,Z=0,W=1),Translation=(X=0,Y=0,Z=0),Scale3D=(X=1,Y=1,Z=1)))
    Item=(Type=Bone,Name="dyn_hair_l_base")
    Primary=(Weight=1,Axis=(X=1,Y=0,Z=0),Target=(X=1,Y=0,Z=0),Kind=Location,Space=(Type=Bone,Name="dyn_hair_l_base"))
    Secondary=(Weight=0,Axis=(X=0,Y=0,Z=-1),Target=(X=1,Y=0,Z=0),Kind=Location,Space=(Type=Bone,Name="head"))
    Weight=1
   ) {is=struct RigVMUnitNode}
   > &RigUnit_AimItem(
    DebugSettings=(Scale=10,WorldOffset=(Rotation=(X=0,Y=0,Z=0,W=1),Translation=(X=0,Y=0,Z=0),Scale3D=(X=1,Y=1,Z=1)))
    Item=(Type=Bone,Name="dyn_hair_r_base")
    Primary=(Weight=1,Axis=(X=1,Y=0,Z=0),Target=(X=1,Y=0,Z=0),Kind=Location,Space=(Type=Bone,Name="dyn_hair_r_base"))
    Secondary=(Weight=0,Axis=(X=0,Y=0,Z=-1),Target=(X=1,Y=0,Z=0),Kind=Location,Space=(Type=Bone,Name="head"))
    Weight=1
   )
   > &RigUnit_ModifyTransforms(
    ItemToModify=((Item=(Type=Bone,Name="dyn_hair_l_base"),Transform=(Rotation=(X=0,Y=0,Z=0,W=1),Translation=(X=0,Y=2,Z=0),Scale3D=(X=0.5,Y=1,Z=0.6))))
    Weight=1
    ItemToModify.0.Transform.Rotation=&RigVMFunction_MathQuaternionFromEuler(Euler=(X=0,Y=-20,Z=5)).Result
   )
   > &RigUnit_ModifyTransforms(
    ItemToModify=((Item=(Type=Bone,Name="dyn_hair_l_01"),Transform=(Rotation=(X=0,Y=0,Z=0,W=1),Translation=(X=0,Y=0,Z=0),Scale3D=(X=0,Y=0,Z=0))))
    Weight=1
    ItemToModify.0.Transform.Rotation=&RigVMFunction_MathQuaternionFromEuler(Euler=(X=0,Y=10,Z=0)).Result
   )
   > &RigUnit_ModifyTransforms(
    ItemToModify=((Item=(Type=Bone,Name="dyn_hair_r_base"),Transform=(Rotation=(X=0,Y=0,Z=0,W=1),Translation=(X=0,Y=0,Z=-2),Scale3D=(X=0.5,Y=1,Z=0.6))))
    Weight=1
    ItemToModify.0.Transform.Rotation=&RigVMFunction_MathQuaternionFromEuler(Euler=(X=20,Y=-20,Z=0)).Result
   )
   > &RigUnit_ModifyTransforms(
    ItemToModify=((Item=(Type=Bone,Name="dyn_hair_r_01"),Transform=(Rotation=(X=0,Y=0,Z=0,W=1),Translation=(X=0,Y=0,Z=0),Scale3D=(X=0,Y=0,Z=0))))
    Weight=1
    ItemToModify.0.Transform.Rotation=&RigVMFunction_MathQuaternionFromEuler(Euler=(X=-40,Y=0,Z=0)).Result
   )
   > &RigUnit_ModifyTransforms(
    ItemToModify=((Item=(Type=Bone,Name="dyn_hair_r_02"),Transform=(Rotation=(X=0,Y=0,Z=0,W=1),Translation=(X=0,Y=0,Z=0),Scale3D=(X=1,Y=1,Z=1))))
    Weight=1
    ItemToModify.0.Transform.Rotation=&RigVMFunction_MathQuaternionFromEuler(Euler=(X=0,Y=-5,Z=0)).Result
   )
   /Script/RigVM.RigVMFunction_AccumulateFloatLerp(
    Blend=16
    InitialValue=1
    bIntegrateDeltaTime=true
    TargetValue=&RigVMFunction_MathFloatRemap(SourceMaximum=1, SourceMinimum=0, TargetMaximum=0.3, TargetMinimum=0, Value=0, bClamp=False).Result
   ) {is=struct RigVMUnitNode, template="AccumulateLerp::Execute(in TargetValue,in InitialValue,in Blend,in bIntegrateDeltaTime,out Result)"}
   &RigVMFunction_AccumulateFloatLerp(
    Blend=16
    InitialValue=1
    bIntegrateDeltaTime=true
    TargetValue=&RigVMFunction_MathFloatRemap(
     SourceMaximum=1
     SourceMinimum=0
     TargetMaximum=0.3
     TargetMinimum=0
     bClamp=false
     Value=/Script/RigVM.RigVMFunction_RandomFloat(Duration=3, Maximum=1, Minimum=0, Seed=217) {is=struct RigVMUnitNode, template="Random::Execute(in Seed,in Minimum,in Maximum,in Duration,out Result)"}.Result
    ).Result
   ) {template="AccumulateLerp::Execute(in TargetValue,in InitialValue,in Blend,in bIntegrateDeltaTime,out Result)"}
   &RigUnit_ModifyTransforms(
    ExecutePin=()
    ItemToModify=((Item=(Type=Bone,Name="dyn_hair_l_02"),Transform=(Rotation=(X=0,Y=0,Z=0,W=1),Translation=(X=0,Y=0,Z=0),Scale3D=(X=1,Y=1,Z=1))))
    Weight=1
    ItemToModify.0.Transform.Rotation=&RigVMFunction_MathQuaternionFromEuler(Euler=(X=0,Y=-5,Z=0)).Result
   )
   &RigUnit_ModifyTransforms(
    ExecutePin=()
    ItemToModify=((Item=(Type=Bone,Name="dyn_hair_r_03"),Transform=(Rotation=(X=0,Y=0,Z=0,W=1),Translation=(X=0,Y=0,Z=0),Scale3D=(X=1,Y=1,Z=1))))
    Weight=1
    ItemToModify.0.Transform.Rotation=&RigVMFunction_MathQuaternionFromEuler(Euler=(X=0,Y=10,Z=0)).Result
   )
  ]}
]
hierarchy [
 root {loc=<0 0 0>, rot=<0 0 0 1>}
 pelvis {parent=root, loc=<0 -0.5301 98.0571>, rot=<0.0127 0.707 -0.0127 -0.707>}
 spine_01 {parent=pelvis, loc=<6.1208 -0.9328 0>, rot=<0 0 0.034 -0.9994>}
 spine_02 {parent=spine_01, loc=<6.5996 -0.2095 0>, rot=<0 0 0.0238 -0.9997>}
 spine_03 {parent=spine_02, loc=<6.4914 -0.1237 0>, rot=<0 0 0.0149 -0.9999>}
 spine_04 {parent=spine_03, loc=<6.9598 0.0794 0>, rot=<0 0 -0.0132 -0.9999>}
 spine_05 {parent=spine_04, loc=<6.9987 0 0>, rot=<0 0 -0.0015 -1>}
 clavicle_l {parent=spine_05, loc=<5.4831 -0.1053 -3.2946>, rot=<0.0465 -0.7289 -0.009 -0.6829>}
 upperarm_l {parent=clavicle_l, loc=<9.8658 -1.2136 -0.4105>, rot=<0.0164 -0.3245 -0.0269 -0.9454>}
 lowerarm_l {parent=upperarm_l, loc=<26.465 -0.8345 0.259>, rot=<0.0036 0.0161 -0.2298 -0.9731>}
 hand_l {parent=lowerarm_l, loc=<22.7355 -0.3203 -0.6105>, rot=<-0.6784 -0.0223 0.0735 -0.7307>}
 thumb_01_l {parent=hand_l, loc=<1.3118 -0.3191 -2.459>, rot=<0.6546 -0.4042 0.0664 -0.6354>}
 thumb_02_l {parent=thumb_01_l, loc=<5.9336 -0.0448 0.0401>, rot=<-0.0037 0.0213 0.298 -0.9543>}
 thumb_03_l {parent=thumb_02_l, loc=<2.2223 0.3682 -0.0907>, rot=<0.0118 -0.013 0.0066 -0.9998>}
 index_metacarpal_l {parent=hand_l, loc=<1.6535 0.0183 -2.96>, rot=<0.0158 -0.0124 -0.0173 -0.9996>}
 index_01_l {parent=index_metacarpal_l, loc=<8.3008 0.0357 -0.0524>, rot=<-0.0044 0.0485 0.0924 -0.9945>}
 index_02_l {parent=index_01_l, loc=<3.3587 -0.3004 0.0452>, rot=<-0.002 0.0027 0.3225 -0.9466>}
 index_03_l {parent=index_02_l, loc=<2.5578 -0.123 -0.0252>, rot=<-0.002 0.0067 0.2558 -0.9667>}
 middle_metacarpal_l {parent=hand_l, loc=<2.1856 -0.1725 -0.5984>, rot=<0.006 0.0168 -0.0329 -0.9993>}
 middle_01_l {parent=middle_metacarpal_l, loc=<7.3275 -0.2447 0.0129>, rot=<-0.0208 0.0042 0.0858 -0.9961>}
 middle_02_l {parent=middle_01_l, loc=<4.0123 0.1401 0.1175>, rot=<-0.0218 0.0258 0.3512 -0.9357>}
 middle_03_l {parent=middle_02_l, loc=<2.6077 0.0534 0.1373>, rot=<0.0082 0.0188 0.1708 -0.9851>}
 ring_metacarpal_l {parent=hand_l, loc=<2.2074 -0.3625 1.3638>, rot=<-0.0284 0.0722 -0.0458 -0.9959>}
 ring_01_l {parent=ring_metacarpal_l, loc=<7.0295 -0.6197 -0.1446>, rot=<-0.0489 -0.0252 0.1892 -0.9804>}
 ring_02_l {parent=ring_01_l, loc=<3.6174 0.4647 -0.1027>, rot=<0 0 0.2806 -0.9598>}
 ring_03_l {parent=ring_02_l, loc=<2.4152 0.114 0.0199>, rot=<0 0 0.1189 -0.9929>}
 pinky_metacarpal_l {parent=hand_l, loc=<1.7616 -0.9753 3.3814>, rot=<-0.0397 0.0814 -0.069 -0.9935>}
 pinky_01_l {parent=pinky_metacarpal_l, loc=<7.1709 -0.4831 -0.1094>, rot=<-0.0654 -0.0035 0.2242 -0.9723>}
 pinky_02_l {parent=pinky_01_l, loc=<2.8274 0.0893 -0.1043>, rot=<0.0037 0.0011 0.2854 -0.9584>}
 pinky_03_l {parent=pinky_02_l, loc=<2.2046 0.0307 -0.0078>, rot=<-0.0002 0.0098 0.1445 -0.9895>}
 lowerarm_twist_02_l {parent=lowerarm_l, loc=<10.6707 0 0>, rot=<0 0 0 1>}
 lowerarm_twist_01_l {parent=lowerarm_l, loc=<18.316 0 0>, rot=<0 0 0 1>}
 upperarm_twist_01_l {parent=upperarm_l, loc=<7.2135 -0.1134 0>, rot=<0 0 0 1>}
 upperarm_twist_02_l {parent=upperarm_l, loc=<19.7998 -0.1134 0>, rot=<0 0 0 1>}
 neck_01 {parent=spine_05, loc=<11.3138 0.1529 0>, rot=<0 0 -0.2003 -0.9797>}
 neck_02 {parent=neck_01, loc=<4.5627 0 0>, rot=<0 0 0.0504 -0.9987>}
 head {parent=neck_02, loc=<5.3914 0 0>, rot=<0 0 0.115 -0.9934>}
 dyn_head_offset {parent=head, loc=<4.6519 0.0385 0>, rot=<0 0 0 1>}
 dyn_ear_l_base {parent=dyn_head_offset, loc=<13.6846 -0.2504 -6.0183>, rot=<-0.6183 0.0489 0.4167 -0.6646>}
 dyn_ear_l_parent {parent=dyn_ear_l_base, loc=<0 0 0>, rot=<0 0 0 1>}
 dyn_ear_l_01 {parent=dyn_ear_l_parent, loc=<7.6574 -1.1352 1.2297>, rot=<0.0578 0.0174 0.0069 -0.9982>}
 dyn_ear_l_02 {parent=dyn_ear_l_01, loc=<6.966 0.6609 1.5339>, rot=<-0.0321 0.0507 -0.1208 -0.9909>}
 dyn_ear_r_base {parent=dyn_head_offset, loc=<13.1319 -0.8914 7.2988>, rot=<-0.6373 0.4341 0.0356 -0.6357>}
 dyn_ear_r_parent {parent=dyn_ear_r_base, loc=<0 0 0>, rot=<0 0 0 1>}
 dyn_ear_r_01 {parent=dyn_ear_r_parent, loc=<5.2847 -0.5095 -0.0678>, rot=<0.0834 0.1048 0.0017 -0.991>}
 dyn_ear_r_02 {parent=dyn_ear_r_01, loc=<7.3652 -1.2137 0.5753>, rot=<-0.0783 0.0047 0.083 -0.9935>}
 dyn_bang_frnt_base {parent=dyn_head_offset, loc=<-2.0484 12.3905 4.4744>, rot=<0.093 -0.7884 0.5949 0.1257>}
 dyn_bang_frnt_parent {parent=dyn_bang_frnt_base, loc=<0 0 0>, rot=<0 0 0 1>}
 dyn_bang_l_base {parent=dyn_head_offset, loc=<-2.8656 1.7812 -14.5306>, rot=<-0.1532 -0.0516 0.9723 -0.1685>}
 dyn_bang_l_parent {parent=dyn_bang_l_base, loc=<0 0 0>, rot=<0 0 0 1>}
 dyn_bang_r_base {parent=dyn_head_offset, loc=<-6.7811 -5.2971 15.9469>, rot=<0.6394 -0.7292 0.1299 0.2062>}
 dyn_bang_r_parent {parent=dyn_bang_r_base, loc=<0 0 0>, rot=<0 0 0 1>}
 dyn_hair_l_base {parent=dyn_head_offset, loc=<-3.3411 -7.6624 -10.2536>, rot=<-0.2155 0.4546 0.8562 -0.1171>}
 dyn_hair_l_parent {parent=dyn_hair_l_base, loc=<0 0 0>, rot=<0 0 0 1>}
 dyn_hair_l_01 {parent=dyn_hair_l_parent, loc=<15.1188 0.8315 0.0017>, rot=<0 0 0 1>}
 dyn_hair_l_02 {parent=dyn_hair_l_01, loc=<16.1889 -0.6213 1.0209>, rot=<0.0305 -0.0174 0.0169 -0.9992>}
 dyn_hair_l_03 {parent=dyn_hair_l_02, loc=<19.4258 0.0165 0.6877>, rot=<-0.0374 -0.017 0.0099 -0.9991>}
 dyn_hair_l_04 {parent=dyn_hair_l_03, loc=<20.9497 0.4382 -0.2684>, rot=<-0.0203 -0.016 0.0118 -0.9996>}
 dyn_hair_r_base {parent=dyn_head_offset, loc=<-4.9067 -7.8794 9.5813>, rot=<0.0285 -0.7849 -0.5693 0.243>}
 dyn_hair_r_parent {parent=dyn_hair_r_base, loc=<0 0 0>, rot=<0 0 0 1>}
 dyn_hair_r_01 {parent=dyn_hair_r_parent, loc=<13.7296 -1.3353 -1.5072>, rot=<-0.0081 -0.0144 0.0294 -0.9994>}
 dyn_hair_r_02 {parent=dyn_hair_r_01, loc=<16.3887 -0.4317 0.9175>, rot=<-0.0458 -0.0192 0.0025 -0.9988>}
 dyn_hair_r_03 {parent=dyn_hair_r_02, loc=<22.5674 -0.2047 0.4831>, rot=<0.0285 0.0026 -0.0083 -0.9996>}
 dyn_hair_r_04 {parent=dyn_hair_r_03, loc=<20.7442 0.101 -1.1297>, rot=<0.014 -0.0251 -0.0304 -0.9991>}
 clavicle_r {parent=spine_05, loc=<5.4831 -0.1053 3.2946>, rot=<0.7289 0.0465 -0.6829 0.009>}
 upperarm_r {parent=clavicle_r, loc=<-9.8658 1.2136 0.4105>, rot=<0.0164 -0.3245 -0.0269 -0.9454>}
 lowerarm_r {parent=upperarm_r, loc=<-26.465 0.8345 -0.259>, rot=<0.0036 0.0161 -0.2298 -0.9731>}
 hand_r {parent=lowerarm_r, loc=<-22.7355 0.3203 0.6106>, rot=<-0.6784 -0.0223 0.0735 -0.7307>}
 thumb_01_r {parent=hand_r, loc=<-1.3118 0.3191 2.459>, rot=<0.6546 -0.4042 0.0664 -0.6354>}
 thumb_02_r {parent=thumb_01_r, loc=<-5.9336 0.0448 -0.0401>, rot=<-0.0037 0.0213 0.298 -0.9543>}
 thumb_03_r {parent=thumb_02_r, loc=<-2.2223 -0.3682 0.0907>, rot=<0.0118 -0.013 0.0066 -0.9998>}
 index_metacarpal_r {parent=hand_r, loc=<-1.6535 -0.0183 2.96>, rot=<0.0158 -0.0124 -0.0173 -0.9996>}
 index_01_r {parent=index_metacarpal_r, loc=<-8.3008 -0.0357 0.0524>, rot=<-0.0044 0.0485 0.0924 -0.9945>}
 index_02_r {parent=index_01_r, loc=<-3.3587 0.3004 -0.0452>, rot=<-0.002 0.0027 0.3225 -0.9466>}
 index_03_r {parent=index_02_r, loc=<-2.5578 0.123 0.0252>, rot=<-0.002 0.0067 0.2558 -0.9667>}
 middle_metacarpal_r {parent=hand_r, loc=<-2.1856 0.1725 0.5984>, rot=<0.006 0.0168 -0.0329 -0.9993>}
 middle_01_r {parent=middle_metacarpal_r, loc=<-7.3275 0.2447 -0.0129>, rot=<-0.0208 0.0042 0.0858 -0.9961>}
 middle_02_r {parent=middle_01_r, loc=<-4.0123 -0.1401 -0.1175>, rot=<-0.0218 0.0258 0.3512 -0.9357>}
 middle_03_r {parent=middle_02_r, loc=<-2.6077 -0.0534 -0.1373>, rot=<0.0082 0.0188 0.1708 -0.9851>}
 ring_metacarpal_r {parent=hand_r, loc=<-2.2074 0.3625 -1.3638>, rot=<-0.0284 0.0722 -0.0458 -0.9959>}
 ring_01_r {parent=ring_metacarpal_r, loc=<-7.0295 0.6197 0.1446>, rot=<-0.0489 -0.0252 0.1892 -0.9804>}
 ring_02_r {parent=ring_01_r, loc=<-3.6174 -0.4647 0.1027>, rot=<0 0 0.2806 -0.9598>}
 ring_03_r {parent=ring_02_r, loc=<-2.4152 -0.114 -0.0199>, rot=<0 0 0.1189 -0.9929>}
 pinky_metacarpal_r {parent=hand_r, loc=<-1.7616 0.9753 -3.3814>, rot=<-0.0397 0.0814 -0.069 -0.9935>}
 pinky_01_r {parent=pinky_metacarpal_r, loc=<-7.1709 0.4831 0.1094>, rot=<-0.0654 -0.0035 0.2242 -0.9723>}
 pinky_02_r {parent=pinky_01_r, loc=<-2.8274 -0.0893 0.1043>, rot=<0.0037 0.0011 0.2854 -0.9584>}
 pinky_03_r {parent=pinky_02_r, loc=<-2.2046 -0.0307 0.0078>, rot=<-0.0002 0.0098 0.1445 -0.9895>}
 lowerarm_twist_02_r {parent=lowerarm_r, loc=<-10.6707 0 0>, rot=<0 0 0 -1>}
 lowerarm_twist_01_r {parent=lowerarm_r, loc=<-18.316 0 0>, rot=<0 0 0 -1>}
 upperarm_twist_01_r {parent=upperarm_r, loc=<-7.2135 0.1134 0>, rot=<0 0 0 -1>}
 upperarm_twist_02_r {parent=upperarm_r, loc=<-19.7998 0.1134 0>, rot=<0 0 0 -1>}
 const_hair_aim {parent=spine_05, loc=<0.3049 -3.7931 0>, rot=<-0.0284 0.7065 0.7065 -0.0284>}
 thigh_l {parent=pelvis, loc=<-1.0855 -0.2109 -8.7053>, rot=<0.0009 0.0355 0.0155 -0.9993>}
 calf_l {parent=thigh_l, loc=<-42.5284 -0.0019 -0.7297>, rot=<0.0006 -0.0161 -0.0391 -0.9991>}
 foot_l {parent=calf_l, loc=<-46.7401 -2.5014 -0.592>, rot=<0.0427 -0.0238 0.0353 -0.9982>}
 ball_l {parent=foot_l, loc=<-3.5658 10.8066 -0.1544>, rot=<0.0001 -0.0001 -0.7063 -0.7079>}
 calf_twist_02_l {parent=calf_l, loc=<-19.1546 -1.7943 -0.2638>, rot=<0.0237 0.0095 -0.0035 -0.9997>}
 calf_twist_01_l {parent=calf_l, loc=<-35.8483 -1.9197 -0.5794>, rot=<0.0237 0.0095 -0.0035 -0.9997>}
 thigh_twist_01_l {parent=thigh_l, loc=<-27.952 0 0>, rot=<0.0013 0 0 -1>}
 thigh_r {parent=pelvis, loc=<-1.0855 -0.2109 8.7053>, rot=<0.0355 -0.0009 0.9993 0.0155>}
 calf_r {parent=thigh_r, loc=<42.5284 0.0019 0.7297>, rot=<0.0006 -0.0161 -0.0391 -0.9991>}
 foot_r {parent=calf_r, loc=<46.7401 2.5014 0.592>, rot=<0.0427 -0.0238 0.0353 -0.9982>}
 ball_r {parent=foot_r, loc=<3.5658 -10.8066 0.1544>, rot=<0.0001 -0.0001 -0.7063 -0.7079>}
 calf_twist_02_r {parent=calf_r, loc=<19.1546 1.7943 0.2638>, rot=<0.0237 0.0095 -0.0035 -0.9997>}
 calf_twist_01_r {parent=calf_r, loc=<35.8483 1.9197 0.5794>, rot=<0.0237 0.0095 -0.0035 -0.9997>}
 thigh_twist_01_r {parent=thigh_r, loc=<27.952 0 0>, rot=<0.0013 0 0 -1>}
]
