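# Physics asset for the long-hair face accessory of the F_MED_Veiled_Glitch female player body, defining how the hair can swing and collide.
# Left and right chains each run from a small base sphere through five capsules, linked by tightly limited angular constraints.
# Swing1 locks on the 01 and 02 joints, and the left tail allows wider angles than the right.
# Unconstrained head, spine, pelvis and upper-arm shapes give the body's collision volume.
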PhysicsAsset /Game/Characters/Player/Female/Medium/Bodies/F_MED_Veiled_Glitch/Meshes/Parts/F_MED_Veiled_Glitch_FaceAcc_Long_Hair_Pack_Physics
bodies {
 dyn_hair_l_01 [{shapeType=Capsule, length=9.1692, radius=4.6102, shapeName=dyn_hair_l_01_capsule, center=<8.2452 0 0>, rotation=<rot 90 0 0>}]
 dyn_hair_l_02 [{shapeType=Capsule, length=12.239, radius=6.1119, shapeName=dyn_hair_l_02_capsule, center=<10.8551 0 0>, rotation=<rot 90 0 0>}]
 dyn_hair_l_03 [{shapeType=Capsule, length=9.4779, radius=9.211, shapeName=dyn_hair_l_03_capsule, center=<11.6801 0 0>, rotation=<rot 90 0 0>}]
 dyn_hair_l_04 [{shapeType=Capsule, length=11.425, radius=12.2708, shapeName=dyn_hair_l_04_capsule, center=<13.7011 0 0>, rotation=<rot 90 0 0>}]
 dyn_hair_l_base [{shapeType=Sphere, radius=0.6009, shapeName=dyn_hair_l_base_sphere}]
 dyn_hair_l_parent [{shapeType=Capsule, length=4.8638, radius=6.5187, shapeName=dyn_hair_l_parent_capsule, center=<8.5603 0 0>, rotation=<rot 90 0 0>}]
 dyn_hair_r_01 [{shapeType=Capsule, length=10.5051, radius=6.4955, shapeName=dyn_hair_r_01_capsule, center=<7.812 0 0>, rotation=<rot 90 0 0>}]
 dyn_hair_r_02 [{shapeType=Capsule, length=13.9156, radius=7.0437, shapeName=dyn_hair_r_02_capsule, center=<13.3232 0 0>, rotation=<rot 90 0 0>}]
 dyn_hair_r_03 [{shapeType=Capsule, length=9.1106, radius=10.3032, shapeName=dyn_hair_r_03_capsule, center=<15.2404 0 0>, rotation=<rot 90 0 0>}]
 dyn_hair_r_04 [{shapeType=Capsule, length=7.7458, radius=12.0883, shapeName=dyn_hair_r_04_capsule, center=<13.607 0 0>, rotation=<rot 90 0 0>}]
 dyn_hair_r_base [{shapeType=Sphere, radius=0.6009, shapeName=dyn_hair_r_base_sphere}]
 dyn_hair_r_parent [{shapeType=Capsule, length=6.7444, radius=6.3983, shapeName=dyn_hair_r_parent_capsule, center=<8.5264 0 0>, rotation=<rot 90 0 0>}]
 dyn_head_offset [{shapeType=Sphere, radius=6.9291, shapeName=dyn_head_offset_sphere}]
 pelvis [{shapeType=Capsule, length=38.2488, radius=18.5572, shapeName=pelvis_capsule, center=<0 3.9815 0>}]
 spine_03 [{shapeType=Capsule, length=13.5402, radius=16.4598, shapeName=spine_03_capsule, center=<0 4.9869 0>, rotation=<rot 90 -179.2838 170.2726>}]
 spine_05 [
  {shapeType=Capsule, length=77.9239, radius=15, shapeName=spine_05_capsule, center=<-1.756 9.1908 0>}
  {shapeType=Capsule, length=4.9899, radius=7.2852, shapeName=spine_05_capsule, center=<4.5915 -10.9677 -1.2389>, rotation=<rot 90 0 -8.9344>}
 ]
 upperarm_l [{shapeType=Capsule, length=21.1331, radius=8.8669, shapeName=upperarm_r_capsule, center=<16.4079 4.3747 0>, rotation=<rot 90 -90 83.6338>}]
 upperarm_r [{shapeType=Capsule, length=21.1331, radius=8.8669, shapeName=upperarm_r_capsule, center=<-16.0963 -7.1677 0>, rotation=<rot 90 0 0>}]
}
constraints [
 {bone1Name=dyn_hair_l_parent, bone2Name=dyn_hair_l_base, swing1LimitDegrees=1, swing2LimitDegrees=8, twistLimitDegrees=1}
 {bone1Name=dyn_hair_l_01, bone2Name=dyn_hair_l_parent, swing1LimitDegrees=3, swing1Motion=Locked, swing2LimitDegrees=10, twistLimitDegrees=3}
 {bone1Name=dyn_hair_l_02, bone2Name=dyn_hair_l_01, swing1LimitDegrees=10, swing1Motion=Locked, swing2LimitDegrees=2, twistLimitDegrees=2}
 {bone1Name=dyn_hair_l_03, bone2Name=dyn_hair_l_02, swing1LimitDegrees=8, swing2LimitDegrees=8, twistLimitDegrees=5}
 {bone1Name=dyn_hair_l_04, bone2Name=dyn_hair_l_03, swing1LimitDegrees=10, swing2LimitDegrees=10, twistLimitDegrees=5}
 {bone1Name=dyn_hair_r_parent, bone2Name=dyn_hair_r_base, swing1LimitDegrees=2, swing2LimitDegrees=8, twistLimitDegrees=1}
 {bone1Name=dyn_hair_r_01, bone2Name=dyn_hair_r_parent, swing1LimitDegrees=3, swing1Motion=Locked, swing2LimitDegrees=10, twistLimitDegrees=3}
 {bone1Name=dyn_hair_r_02, bone2Name=dyn_hair_r_01, swing1LimitDegrees=5, swing1Motion=Locked, swing2LimitDegrees=2, twistLimitDegrees=2}
 {bone1Name=dyn_hair_r_03, bone2Name=dyn_hair_r_02, swing1LimitDegrees=5, swing2LimitDegrees=5, twistLimitDegrees=3}
 {bone1Name=dyn_hair_r_04, bone2Name=dyn_hair_r_03, swing1LimitDegrees=5, swing2LimitDegrees=5, twistLimitDegrees=3}
]
