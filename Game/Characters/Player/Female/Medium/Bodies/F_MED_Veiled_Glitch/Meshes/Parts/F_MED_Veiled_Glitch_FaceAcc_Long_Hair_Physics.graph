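# Physics asset for the long-hair face accessory of the female medium player body Veiled Glitch, covering left and right dynamic hair chains.
# Each chain links a base sphere to five capsules, mostly widening toward the tip, through constraints with swing and twist limits.
# Unconstrained head, neck, spine, pelvis and upper-arm shapes supply body volumes alongside the hair.
# Limits are asymmetric: the right chain has much tighter second-swing limits than the left.
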PhysicsAsset /Game/Characters/Player/Female/Medium/Bodies/F_MED_Veiled_Glitch/Meshes/Parts/F_MED_Veiled_Glitch_FaceAcc_Long_Hair_Physics
bodies {
 dyn_hair_l_01 [{shapeType=Capsule, length=9.4217, radius=4.3577, shapeName=dyn_hair_l_01_capsule, center=<8.2452 0 0>, rotation=<rot 90 0 0>}]
 dyn_hair_l_02 [{shapeType=Capsule, length=12.8727, radius=5.4782, shapeName=dyn_hair_l_02_capsule, center=<10.8551 0 0>, rotation=<rot 90 0 0>}]
 dyn_hair_l_03 [{shapeType=Capsule, length=6.982, radius=11.7068, shapeName=dyn_hair_l_03_capsule, center=<11.6801 0 0>, rotation=<rot 90 0 0>}]
 dyn_hair_l_04 [{shapeType=Capsule, length=9.8432, radius=13.8526, shapeName=dyn_hair_l_04_capsule, center=<17.245 0 0>, rotation=<rot 90 0 0>}]
 dyn_hair_l_base [{shapeType=Sphere, radius=0.6009, shapeName=dyn_hair_l_base_sphere}]
 dyn_hair_l_parent [{shapeType=Capsule, length=6.817, radius=4.5655, shapeName=dyn_hair_l_parent_capsule, center=<8.5603 0 0>, rotation=<rot 90 0 0>}]
 dyn_hair_r_01 [{shapeType=Capsule, length=12.2746, radius=4.726, shapeName=dyn_hair_r_01_capsule, center=<7.812 0 0>, rotation=<rot 90 0 0>}]
 dyn_hair_r_02 [{shapeType=Capsule, length=17.174, radius=6.5858, shapeName=dyn_hair_r_02_capsule, center=<13.3232 0 0>, rotation=<rot 90 0 0>}]
 dyn_hair_r_03 [{shapeType=Capsule, length=7.5467, radius=11.8671, shapeName=dyn_hair_r_03_capsule, center=<16.0182 0 0>, rotation=<rot 90 0 0>}]
 dyn_hair_r_04 [{shapeType=Capsule, length=6.5448, radius=13.2893, shapeName=dyn_hair_r_04_capsule, center=<16.6421 0 0>, rotation=<rot 90 0 0>}]
 dyn_hair_r_base [{shapeType=Sphere, radius=0.6009, shapeName=dyn_hair_r_base_sphere}]
 dyn_hair_r_parent [{shapeType=Capsule, length=7.6172, radius=5.5255, shapeName=dyn_hair_r_parent_capsule, center=<8.5264 0 0>, rotation=<rot 90 0 0>}]
 dyn_head_offset [{shapeType=Sphere, radius=6.9291, shapeName=dyn_head_offset_sphere}]
 neck_02 [{shapeType=Capsule, length=9.6534, radius=6.3383, shapeName=neck_02_capsule, center=<-3.3893 -0.5242 0>, rotation=<rot 90 0 0>}]
 pelvis [{shapeType=Capsule, length=40.5507, radius=16.2553, shapeName=pelvis_capsule, center=<0 3.9815 0>}]
 spine_03 [{shapeType=Capsule, length=13.5402, radius=16.4598, shapeName=spine_03_capsule, center=<0 4.9869 0>, rotation=<rot 90 -179.2838 170.2726>}]
 spine_05 [{shapeType=Capsule, length=28.0984, radius=15, shapeName=spine_05_capsule, center=<-4.7052 3.3367 0>}]
 upperarm_l [{
   shapeType=Capsule
   length=20.8647
   radius=9.1353
   shapeName=upperarm_l_capsule
   center=<16.6811 5.3334 0>
   rotation=<rot 90 -90.9147 -90.9896>
  }]
 upperarm_r [{shapeType=Capsule, length=20.8647, radius=9.1353, shapeName=upperarm_l_capsule, center=<-16.4465 -4.2901 0>, rotation=<rot 90 180 -0.4629>}]
}
constraints [
 {bone1Name=dyn_hair_l_parent, bone2Name=dyn_hair_l_base, swing1LimitDegrees=2, swing2LimitDegrees=20, twistLimitDegrees=5}
 {bone1Name=dyn_hair_l_01, bone2Name=dyn_hair_l_parent, swing1LimitDegrees=10, swing2LimitDegrees=30, twistLimitDegrees=5}
 {bone1Name=dyn_hair_l_02, bone2Name=dyn_hair_l_01, swing1LimitDegrees=15, swing2LimitDegrees=15, twistLimitDegrees=5}
 {bone1Name=dyn_hair_l_03, bone2Name=dyn_hair_l_02, swing1LimitDegrees=15, swing2LimitDegrees=15, twistLimitDegrees=5}
 {bone1Name=dyn_hair_l_04, bone2Name=dyn_hair_l_03, swing1LimitDegrees=15, swing2LimitDegrees=15, twistLimitDegrees=5}
 {bone1Name=dyn_hair_r_parent, bone2Name=dyn_hair_r_base, swing1LimitDegrees=2, swing2LimitDegrees=2, twistLimitDegrees=5}
 {bone1Name=dyn_hair_r_01, bone2Name=dyn_hair_r_parent, swing1LimitDegrees=10, swing2LimitDegrees=2, twistLimitDegrees=5}
 {bone1Name=dyn_hair_r_02, bone2Name=dyn_hair_r_01, swing1LimitDegrees=15, swing2LimitDegrees=2, twistLimitDegrees=5}
 {bone1Name=dyn_hair_r_03, bone2Name=dyn_hair_r_02, swing1LimitDegrees=15, swing2LimitDegrees=5, twistLimitDegrees=5}
 {bone1Name=dyn_hair_r_04, bone2Name=dyn_hair_r_03, swing1LimitDegrees=15, swing2LimitDegrees=10, twistLimitDegrees=5}
]
