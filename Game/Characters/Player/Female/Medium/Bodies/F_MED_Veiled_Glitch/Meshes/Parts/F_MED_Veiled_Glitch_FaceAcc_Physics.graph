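# Physics asset for the Veiled Glitch female player character's face accessory mesh, covering dynamic ear and bang bones.
# Each ear is a three-capsule chain on a small base sphere, and the front, left and right bangs are single capsules on their own base spheres.
# Tight swing and twist limits link each segment to its parent, with differing left and right ear limits.
# A large, unconstrained sphere sits on the head offset bone.
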PhysicsAsset /Game/Characters/Player/Female/Medium/Bodies/F_MED_Veiled_Glitch/Meshes/Parts/F_MED_Veiled_Glitch_FaceAcc_Physics
bodies {
 dyn_bang_frnt_base [{shapeType=Sphere, radius=0.9432, shapeName=dyn_bang_frnt_base_sphere}]
 dyn_bang_frnt_parent [{shapeType=Capsule, length=2.4076, radius=2.3682, shapeName=dyn_bang_frnt_parent_capsule, center=<2.4889 0 0>, rotation=<rot 90 0 0>}]
 dyn_bang_l_base [{shapeType=Sphere, radius=0.5, shapeName=dyn_bang_l_base_sphere}]
 dyn_bang_l_parent [{shapeType=Capsule, length=2.2933, radius=2.0449, shapeName=dyn_bang_l_parent_capsule, center=<2.3299 0 0>, rotation=<rot 90 0 0>}]
 dyn_bang_r_base [{shapeType=Sphere, radius=0.9432, shapeName=dyn_bang_r_base_sphere}]
 dyn_bang_r_parent [{shapeType=Capsule, length=1.4648, radius=3.0252, shapeName=dyn_bang_r_parent_capsule, center=<1.7626 0 0>, rotation=<rot 90 0 0>}]
 dyn_ear_l_01 [{shapeType=Capsule, length=3.4247, radius=2.9904, shapeName=dyn_ear_l_01_capsule, center=<2.5248 0 0>, rotation=<rot 90 0 0>}]
 dyn_ear_l_02 [{shapeType=Capsule, length=2.7115, radius=2.5413, shapeName=dyn_ear_l_02_capsule, center=<2.7407 0 0>, rotation=<rot 90 0 0>}]
 dyn_ear_l_base [{shapeType=Sphere, radius=0.5, shapeName=dyn_ear_l_base_sphere}]
 dyn_ear_l_parent [{shapeType=Capsule, length=3.2303, radius=2.7237, shapeName=dyn_ear_l_parent_capsule, center=<3.3803 0 0>, rotation=<rot 90 0 0>}]
 dyn_ear_r_01 [{shapeType=Capsule, length=2.808, radius=3.1917, shapeName=dyn_ear_r_01_capsule, center=<3.1334 0 0>, rotation=<rot 90 0 0>}]
 dyn_ear_r_02 [{shapeType=Capsule, length=4.2196, radius=3.2598, shapeName=dyn_ear_r_02_capsule, center=<3.2364 0 0>, rotation=<rot 90 0 0>}]
 dyn_ear_r_base [{shapeType=Sphere, radius=0.5, shapeName=dyn_ear_r_base_sphere}]
 dyn_ear_r_parent [{shapeType=Capsule, length=2.4312, radius=3.0122, shapeName=dyn_ear_r_parent_capsule, center=<3.1027 0 0>, rotation=<rot 90 0 0>}]
 dyn_head_offset [{shapeType=Sphere, radius=6.9291, shapeName=dyn_head_offset_sphere}]
}
constraints [
 {bone1Name=dyn_ear_l_parent, bone2Name=dyn_ear_l_base, swing1LimitDegrees=2, swing2LimitDegrees=2, twistLimitDegrees=2}
 {bone1Name=dyn_ear_l_01, bone2Name=dyn_ear_l_parent, swing1LimitDegrees=8, swing2LimitDegrees=4, twistLimitDegrees=2}
 {bone1Name=dyn_ear_r_parent, bone2Name=dyn_ear_r_base, swing1LimitDegrees=10, swing2LimitDegrees=8, twistLimitDegrees=2}
 {bone1Name=dyn_ear_r_01, bone2Name=dyn_ear_r_parent, swing1LimitDegrees=8, swing2LimitDegrees=12, twistLimitDegrees=2}
 {bone1Name=dyn_bang_r_parent, bone2Name=dyn_bang_r_base, swing1LimitDegrees=8, swing2LimitDegrees=8, twistLimitDegrees=3}
 {bone1Name=dyn_bang_l_parent, bone2Name=dyn_bang_l_base, swing1LimitDegrees=8, swing2LimitDegrees=8, twistLimitDegrees=3}
 {bone1Name=dyn_bang_frnt_parent, bone2Name=dyn_bang_frnt_base, swing1LimitDegrees=10, swing2LimitDegrees=10, twistLimitDegrees=3}
 {bone1Name=dyn_ear_l_02, bone2Name=dyn_ear_l_01, swing1LimitDegrees=10, swing2LimitDegrees=6, twistLimitDegrees=2}
 {bone1Name=dyn_ear_r_02, bone2Name=dyn_ear_r_01, swing1LimitDegrees=12, swing2LimitDegrees=14, twistLimitDegrees=2}
]
